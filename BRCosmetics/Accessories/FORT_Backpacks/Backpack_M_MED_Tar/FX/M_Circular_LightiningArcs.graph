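# Translucent unlit Niagara sprite material for circular lightning arcs in the Tar backpack effects, tinted by particle color.
# Muzzle-ring texture channels map onto tiled radial coordinates, with tiling and erosion driven by dynamic parameters.
# Floored, sine-modulated particle time feeds the panner, so arcs jump in steps rather than scroll smoothly.
# Subtracting a ring mask plus radial falloff shapes opacity, scaled by particle alpha.
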Material /BRCosmetics/Accessories/FORT_Backpacks/Backpack_M_MED_Tar/FX/M_Circular_LightiningArcs
graph {
 properties {
  bUsedWithNiagaraSprites=true
  blendMode=BLEND_Translucent
  pixelDepthOffsetMode=PDOM_Legacy
  shadingModel=MSM_Unlit
  twoSided=true
  usedShadingModels=Unlit
 }
 nodes [
  n1 = ParticleColor()
  MaterialFunctionCall(Emissive=@n1.RGB) {materialFunction=&MF_FX_EmissiveAdjustments}
  n2 = Multiply(
   A=ScalarParameter() {defaultValue=1, parameterName=U Tile}
   B=DynamicParameter() {defaultValue=<rgba 1 0 1 1>, paramNames [Panner Speed, Erode, V Tile, U Tile]}.U Tile
  )
  n3 = Multiply(
   A=DynamicParameter() {defaultValue=<rgba 1 0 1 1>, paramNames [Panner Speed, Erode, V Tile, U Tile]}.V Tile
   B=ScalarParameter() {defaultValue=2, parameterName=V Tile}
  ) {constA=1}
  n4 = Add(
   A=Multiply(
    A=MaterialFunctionCall() {materialFunction=&VectorToRadialValue}.Radial Coordinates
    B=AppendVector(A=@n2, B=@n3)
   )
   B=Constant2Vector() {g=-0.5}
  )
  n5 = ParticleRelativeTime()
  n6 = Sine(None=@n5) {period=0.77}
  n7 = Sine(None=@n5) {period=1.12}
  n8 = DynamicParameter() {defaultValue=<rgba 1 0 1 1>, paramNames [Panner Speed, Erode, V Tile, U Tile]}
  n9 = Add(A=Abs(None=Add(A=@n6, B=@n7)), B=Add(A=Multiply(A=@n5, B=@n8.Panner Speed), B=@n8.Panner Speed))
  n10 = VectorParameter() {parameterName=PannerSpeed, defaultValue=<rgba 0 1.25 0 0>}
  n11 = TextureSample!(
   UVs=Panner(
    Coordinate=MaterialFunctionCall(XY=@n4) {materialFunction=&Swizzle}.YX
    Time=Floor(None=@n9)
    Speed=ComponentMask(None=@n10.RGB) {g=true, r=true}
   ) {bFractionalPart=true, speedY=1.25}
   Tex
   Level
   Apply View MipBias
  ) {mipValueMode=TMVM_MipLevel, samplerType=SAMPLERTYPE_LinearColor, texture=/Game/Effects/Fort_Effects/Textures/Tile/T_AMR_MuzzleRing_01}
  n12 = Multiply(A=ParticleRelativeTime(), B=ScalarParameter() {defaultValue=2, parameterName=PRT Scale})
  n13 = Multiply(
   A=LinearInterpolate(A=@n11.B, B=@n11.B, Alpha=Clamp(None=@n12))
   B=Add(A=@n11.G, B=DynamicParameter() {defaultValue=<rgba 1 0 1 1>, paramNames [Panner Speed, Erode, V Tile, U Tile]}.Erode)
  )
  n14 = MaterialFunctionCall() {materialFunction=&VectorToRadialValue}
  n15 = ComponentMask(None=MaterialFunctionCall() {materialFunction=&VectorToRadialValue}.Radial Coordinates) {r=true}
  n16 = Add(
   A=TextureSample!(
    UVs=MaterialFunctionCall(XY=@n14.Radial Coordinates) {materialFunction=&Swizzle}.YX
    Tex
    Apply View MipBias
   ) {samplerType=SAMPLERTYPE_LinearColor, texture=&T_AMR_MuzzleRing_01}.A
   B=Power(Base=OneMinus(None=@n15)) {constExponent=4}
  )
  Multiply(A=@n1.A, B=Clamp(None=Subtract(A=@n13, B=@n16)))
 ]
}
